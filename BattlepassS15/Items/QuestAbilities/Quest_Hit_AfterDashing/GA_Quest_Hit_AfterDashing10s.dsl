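; Gameplay ability that credits a Battle Pass S15 quest for a hit made after dashing, judging by its name a 10-second window.
; On event activation it checks the owning player pawn for the after-dash tag, ending immediately if absent.
; A Blocked flag ignores re-entry during a 0.2-second delay.
; It then sends the hit stat with target tags to the quest manager, removes the effect granting the after-dash tag and ends.
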(event Ability|EventActivateAbilityFromEvent
  (params (EventData :type Gameplay%20Event%20Data%20Structure%20%28by%20ref%29))
  (bind (_eventtag _instigator _target _optionalobject _optionalobject2 _contexthandle _instigatortags _targettags) (Utilities|Struct|BreakGameplayEventData EventData))
  (|CallParentFunction EventData)
  (if (GameplayTags|HasTag (GameplayTags|GetOwnedGameplayTags (Variables|GAVars|Data|GetOwningPlayerPawn) :!class /Script/GameplayTags.GameplayTagAssetInterface) (Variables|Default|GetT_AfterDash10s) true)
    (if (not (Variables|Default|GetBlocked))
      (Variables|Default|SetBlocked true)
      (Utilities|FlowControl|Delay 0.2)
      (Variables|Default|SetBlocked false)
      (Quest|SendComplexCustomStatEvent (Variables|GAVars|Data|GetOwningPlayerQuestManager) 0 (Variables|Default|GetT_HitAfterDashing) _targettags 1)
      (Ability|RemoveGameplayEffectFromOwnerWithGrantedTags "(GameplayTags=((TagName=\"Granted.Athena.Quest.AfterDashing10s\")))")
      (Ability|EndAbility))
    (else
      (Ability|EndAbility))))
